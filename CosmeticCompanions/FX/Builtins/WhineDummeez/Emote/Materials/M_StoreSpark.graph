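# Translucent unlit spark material for Niagara and particle sprites in the WhineDummeez cosmetic companion emote effects.
# Particle color RGB runs through the MF_FX_EmissiveAdjustments function for the glow.
# The T_Streak texture, borrowed from the level-up UI effects, is multiplied by particle alpha to shape opacity, with a depth fade softening intersections with geometry.
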Material /CosmeticCompanions/FX/Builtins/WhineDummeez/Emote/Materials/M_StoreSpark
graph {
 properties {
  bUsedWithNiagaraSprites=true
  bUsedWithParticleSprites=true
  blendMode=BLEND_Translucent
  pixelDepthOffsetMode=PDOM_Legacy
  shadingModel=MSM_Unlit
  usedShadingModels=Unlit
 }
 nodes [
  n1 = ParticleColor()
  DepthFade(Opacity=Multiply(A=@n1.A, B=TextureSample() {texture=/Game/Effects/Fort_Effects/UI/LevelUp/T_Streak}.RGB)) {fadeDistanceDefault=75}
  MaterialFunctionCall(Emissive=@n1.RGB) {materialFunction=&MF_FX_EmissiveAdjustments}
 ]
}
