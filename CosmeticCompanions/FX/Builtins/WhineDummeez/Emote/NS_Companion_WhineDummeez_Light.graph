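# Niagara system for the light effect of the WhineDummeez cosmetic companion's emote, combining a red head light with a warm flare.
# HeadLight bursts a particle drawn as mesh, sprite and dynamic light, with randomized flicker curves driving halo and inner-flare colors, all multiplied by System.Fade.
# Flare runs on its own delayed emitter timing, using non-uniform size and flickering color curves to swell, spike and vanish.
# HeadLight001 is a disabled near-copy of HeadLight with an extra sprite renderer.
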NiagaraSystem /CosmeticCompanions/FX/Builtins/WhineDummeez/Emote/NS_Companion_WhineDummeez_Light
dependencies {
 usedDataInterfaces [&NiagaraDataInterfaceColorCurve, &NiagaraDataInterfaceCurve, &NiagaraDataInterfaceVector2DCurve]
 usedDynamicInputs [&Vector4FromCurve]
 usedModules [
  &SystemState
  &self:SystemSpawnScript.SystemScriptSource.SystemScriptGraph.NiagaraNodeAssignment_0.SetVariables_#1
 ]
}
emitters {
 Flare {
  emitterUpdateScript [
   &EmitterState(Life Cycle Mode=::NewEnumerator1, Loop Behavior=::NewEnumerator1, Loop Delay=0.4, Loop Duration=3, UseLoopDelay=true)
   &SpawnBurst_Instantaneous(Use Spawn Probability=false, UseLoopCountLimit=false)
  ]
  particleSpawnScript [&InitializeParticle(
    Color=<rgba 0.7865 0.4898 0.1373 1>
    Lifetime=2.5
    Position Offset=<0 0 2>
    Sprite Size Mode=::NewEnumerator3
    UsePositionOffset=true
    Write Material Random=true
   )]
  particleUpdateScript [
   &ParticleState()
   &ScaleSpriteSize(
    Non-Uniform Curve Sprite Scale=DataInterface() {propertyValues {
      XCurve {
       defaultValue=0
       keys [
        {arriveTangent=1.7915, interpMode=RCIM_Cubic, leaveTangent=1.7915, tangentMode=RCTM_User, time=0, value=0}
        {arriveTangent=1.912, interpMode=RCIM_Cubic, leaveTangent=1.912, tangentMode=RCTM_User, time=0.5971, value=1.9372}
        {arriveTangent=-16.9014, interpMode=RCIM_Cubic, leaveTangent=-16.9014, tangentMode=RCTM_User, time=0.6486, value=4}
        {arriveTangent=-15.7653, interpMode=RCIM_Cubic, leaveTangent=-15.7653, tangentMode=RCTM_User, time=0.7484, value=1.9919}
        {arriveTangent=-8.1193, interpMode=RCIM_Cubic, leaveTangent=-8.1193, tangentMode=RCTM_User, time=0.9795, value=0}
       ]
      }
      YCurve {
       defaultValue=0
       keys [
        {arriveTangent=0.9165, interpMode=RCIM_Cubic, leaveTangent=0.9165, tangentMode=RCTM_User, time=0, value=0}
        {arriveTangent=1.5123, interpMode=RCIM_Cubic, leaveTangent=1.5123, time=0.5971, value=1.9222}
        {arriveTangent=0.0038, interpMode=RCIM_Cubic, leaveTangent=0.0038, time=0.6836, value=1.0338}
        {arriveTangent=-15.8322, interpMode=RCIM_Cubic, leaveTangent=-15.8322, tangentMode=RCTM_User, time=0.7462, value=1.9228}
        {arriveTangent=-8.1193, interpMode=RCIM_Cubic, leaveTangent=-8.1193, tangentMode=RCTM_User, time=0.9795, value=0}
       ]
      }
     }}
    Scale Sprite Size Mode=::NewEnumerator3
   )
   &ScaleColor(
    Scale Alpha=&FloatFromCurve(FloatCurve=DataInterface() {propertyValues {Curve {
        defaultValue=1
        keys [
         {interpMode=RCIM_Cubic, tangentMode=RCTM_User, time=0, value=0}
         {interpMode=RCIM_Cubic, tangentMode=RCTM_User, time=0.7859, value=0.9806}
         {tangentMode=RCTM_User, time=1, value=0}
        ]
       }}})
    Scale RGB=&VectorFromFloat(Value=&FloatFromCurve(
      FloatCurve=DataInterface() {propertyValues {Curve {
         defaultValue=1
         keys [
          {interpMode=RCIM_Cubic, tangentMode=RCTM_User, time=0, value=1}
          {interpMode=RCIM_Cubic, tangentMode=RCTM_User, time=0.6, value=2}
          {tangentMode=RCTM_User, time=0.869, value=0}
         ]
        }}}
      Scale Curve=&FloatFromCurve(
       FloatCurve=DataInterface() {propertyValues {Curve {keys [
           <at 0 1>
           <at 0.5463 1>
           <at 0.6584 1.5923>
           <at 0.7373 1>
           {interpMode=RCIM_Cubic, tangentMode=RCTM_User, time=0.8346, value=1.5923}
           <at 0.9087 1>
           <at 1 1>
          ]}}}
       Scale Curve=&RandomRangeFloat(Minimum=0.6, Maximum=1, Evaluation Type=::NewEnumerator1)
      )
     ))
    ScaleA=true
    ScaleRGB=true
   )
   &SolveForcesAndVelocity(Clamp Velocity=false, Limit Acceleration=false)
   &CameraOffset(Camera Offset Amount=15)
  ]
  rendererClasses [&NiagaraSpriteRendererProperties]
  renderers [{rendererClass=&NiagaraSpriteRendererProperties}]
 }
 HeadLight {
  emitterUpdateScript [
   &EmitterState()
   &SpawnBurst_Instantaneous(Use Spawn Probability=false, UseLoopCountLimit=false)
  ]
  particleSpawnScript [
   &InitializeParticle(
    Color=<rgba 1 0.0326 0 1>
    Lifetime=5
    Mesh Scale=&Multiply_VectorByFloat(Vector=<1 1 1>, Float=0.08)
    Mesh Scale Mode=::NewEnumerator4
    Position Offset=<0 0 2>
    Sprite Size Mode=::NewEnumerator3
    Uniform Sprite Size=25
    UsePositionOffset=true
    Write Material Random=true
   )
   &InitialMeshOrientation(Mesh Axis To Orient=<0 0 -1>, Mesh Orientation Mode=::NewEnumerator2, Use Rotation Vector=false)
  ]
  particleUpdateScript [
   &ParticleState()
   &ScaleColor(
    Scale Alpha=&FloatFromCurve(FloatCurve=DataInterface() {propertyValues {Curve {
        defaultValue=0
        keys [
         {arriveTangent=4.1785, interpMode=RCIM_Cubic, leaveTangent=4.1785, tangentMode=RCTM_User, time=0, value=0}
         {arriveTangent=1.1179, interpMode=RCIM_Cubic, leaveTangent=1.1179, tangentMode=RCTM_User, time=0.2042, value=0.6228}
         {arriveTangent=0.0224, interpMode=RCIM_Cubic, leaveTangent=0.0224, tangentMode=RCTM_User, time=1, value=1}
        ]
       }}})
    ScaleA=true
    ScaleRGB=false
   )
   &SolveForcesAndVelocity(Clamp Velocity=false, Limit Acceleration=false)
   /SetVariables(
    Particles.HaloColor:LinearColor=&Multiply_LinearColorByFloat(
     Linear Color=$Particles.Color
     Float=&FloatFromCurve(
      FloatCurve=DataInterface() {propertyValues {Curve {keys [
          <at 0 1>
          <at 0.0514 0>
          <at 0.0629 2>
          <at 0.0909 1>
          <at 0.2042 1>
          <at 0.2411 0.5>
          <at 0.2749 2>
          <at 0.3179 1>
          <at 0.407 1>
          <at 0.4531 2>
          <at 0.4961 1>
          {arriveTangent=2.4659, leaveTangent=2.4659, time=0.6375, value=1}
          <at 0.6713 0>
          <at 0.699 2>
          <at 0.7389 1>
          <at 0.8587 1>
          <at 0.8987 2>
          <at 0.9417 1>
          <at 1 1>
         ]}}}
      Scale Curve=&RandomRangeFloat(Minimum=1.7, Maximum=2.7, Evaluation Type=::NewEnumerator1)
     )
    )
    Particles.InnerFlareColor:LinearColor=&Multiply_LinearColorByFloat(
     Linear Color=$Particles.Color
     Float=&FloatFromCurve(
      FloatCurve=DataInterface() {propertyValues {Curve {keys [
          <at 0 1>
          <at 0.0514 0>
          <at 0.0629 2>
          <at 0.0909 1>
          <at 0.2042 1>
          <at 0.2411 0.5>
          <at 0.2749 2>
          <at 0.3179 1>
          <at 0.407 1>
          <at 0.4531 2>
          <at 0.4961 1>
          {arriveTangent=2.4659, leaveTangent=2.4659, time=0.6375, value=1}
          <at 0.6713 0>
          <at 0.699 2>
          <at 0.7389 1>
          <at 0.8587 1>
          <at 0.8987 2>
          <at 0.9417 1>
          <at 1 1>
         ]}}}
      Scale Curve=&RandomRangeFloat(Minimum=2, Maximum=3, Evaluation Type=::NewEnumerator1)
     )
    )
   ) {isSetParameters=true, module=SetVariables_#2}
   /SetVariables(
    Particles.Color=&Multiply_LinearColorByFloat(Linear Color=$Particles.Color, Float=$System.Fade)
    Particles.HaloColor:LinearColor=&Multiply_LinearColorByFloat(Linear Color=$Particles.HaloColor, Float=$System.Fade)
    Particles.InnerFlareColor:LinearColor=&Multiply_LinearColorByFloat(Linear Color=$Particles.InnerFlareColor, Float=$System.Fade)
   ) {isSetParameters=true, module=SetVariables_#3}
   &FN_Light_Attributes(Create Exponent Attribute=true, Create Radius Attribute=true, Radius=75)
  ]
  rendererClasses [
   &NiagaraMeshRendererProperties
   &NiagaraSpriteRendererProperties
   &NiagaraLightRendererProperties
  ]
  renderers [
   {rendererClass=&NiagaraMeshRendererProperties, rendererIndex=0}
   {rendererClass=&NiagaraSpriteRendererProperties, rendererIndex=1}
   {rendererClass=&NiagaraLightRendererProperties, rendererIndex=2}
  ]
 }
 HeadLight001 {
  bEnabled=false
  emitterUpdateScript [
   &EmitterState()
   &SpawnBurst_Instantaneous(Use Spawn Probability=false, UseLoopCountLimit=false)
  ]
  particleSpawnScript [
   &InitializeParticle(
    Color=<rgba 1 0.0326 0 1>
    Lifetime=5
    Mesh Scale=&Multiply_VectorByFloat(Vector=<1 1 1>, Float=0.08)
    Mesh Scale Mode=::NewEnumerator4
    Position Offset=<0 0 2>
    Sprite Size Mode=::NewEnumerator3
    Uniform Sprite Size=25
    UsePositionOffset=true
    Write Material Random=true
   )
   &InitialMeshOrientation(Mesh Axis To Orient=<0 0 -1>, Mesh Orientation Mode=::NewEnumerator2, Use Rotation Vector=false)
  ]
  particleUpdateScript [
   &ParticleState()
   &ScaleColor(
    Scale Alpha=&FloatFromCurve(FloatCurve=DataInterface() {propertyValues {Curve {
        defaultValue=0
        keys [
         {arriveTangent=4.1785, interpMode=RCIM_Cubic, leaveTangent=4.1785, tangentMode=RCTM_User, time=0, value=0}
         {arriveTangent=1.1179, interpMode=RCIM_Cubic, leaveTangent=1.1179, tangentMode=RCTM_User, time=0.2042, value=0.6228}
         {arriveTangent=0.0224, interpMode=RCIM_Cubic, leaveTangent=0.0224, tangentMode=RCTM_User, time=1, value=1}
        ]
       }}})
    ScaleA=true
    ScaleRGB=false
   )
   &SolveForcesAndVelocity(Clamp Velocity=false, Limit Acceleration=false)
   /SetVariables(
    Particles.HaloColor:LinearColor=&Multiply_LinearColorByFloat(
     Linear Color=$Particles.Color
     Float=&FloatFromCurve(
      FloatCurve=DataInterface() {propertyValues {Curve {keys [
          <at 0 1>
          <at 0.0514 0>
          <at 0.0629 2>
          <at 0.0909 1>
          <at 0.2042 1>
          <at 0.2411 0.5>
          <at 0.2749 2>
          <at 0.3179 1>
          <at 0.407 1>
          <at 0.4531 2>
          <at 0.4961 1>
          {arriveTangent=2.4659, leaveTangent=2.4659, time=0.6375, value=1}
          <at 0.6713 0>
          <at 0.699 2>
          <at 0.7389 1>
          <at 0.8587 1>
          <at 0.8987 2>
          <at 0.9417 1>
          <at 1 1>
         ]}}}
      Scale Curve=&RandomRangeFloat(Minimum=1.7, Maximum=2.7, Evaluation Type=::NewEnumerator1)
     )
    )
    Particles.InnerFlareColor:LinearColor=&Multiply_LinearColorByFloat(
     Linear Color=$Particles.Color
     Float=&FloatFromCurve(
      FloatCurve=DataInterface() {propertyValues {Curve {keys [
          <at 0 1>
          <at 0.0514 0>
          <at 0.0629 2>
          <at 0.0909 1>
          <at 0.2042 1>
          <at 0.2411 0.5>
          <at 0.2749 2>
          <at 0.3179 1>
          <at 0.407 1>
          <at 0.4531 2>
          <at 0.4961 1>
          {arriveTangent=2.4659, leaveTangent=2.4659, time=0.6375, value=1}
          <at 0.6713 0>
          <at 0.699 2>
          <at 0.7389 1>
          <at 0.8587 1>
          <at 0.8987 2>
          <at 0.9417 1>
          <at 1 1>
         ]}}}
      Scale Curve=&RandomRangeFloat(Minimum=2, Maximum=3, Evaluation Type=::NewEnumerator1)
     )
    )
   ) {isSetParameters=true, module=SetVariables_#2}
   /SetVariables(
    Particles.Color=&Multiply_LinearColorByFloat(Linear Color=$Particles.Color, Float=$System.Fade)
    Particles.HaloColor:LinearColor=&Multiply_LinearColorByFloat(Linear Color=$Particles.HaloColor, Float=$System.Fade)
    Particles.InnerFlareColor:LinearColor=&Multiply_LinearColorByFloat(Linear Color=$Particles.InnerFlareColor, Float=$System.Fade)
   ) {isSetParameters=true, module=SetVariables_#3}
   &FN_Light_Attributes(Create Exponent Attribute=true, Create Radius Attribute=true, Radius=75)
  ]
  rendererClasses [
   &NiagaraMeshRendererProperties
   &NiagaraSpriteRendererProperties
   &NiagaraLightRendererProperties
  ]
  renderers [
   {rendererClass=&NiagaraMeshRendererProperties, rendererIndex=0}
   {rendererClass=&NiagaraSpriteRendererProperties, rendererIndex=1}
   {rendererClass=&NiagaraSpriteRendererProperties, rendererIndex=2}
   {rendererClass=&NiagaraLightRendererProperties, rendererIndex=3}
  ]
 }
}
summary {
 systemName=NS_Companion_WhineDummeez_Light
 emitters [
  {
   emitterName=HeadLight
   rendererClasses [
    &NiagaraMeshRendererProperties
    &NiagaraSpriteRendererProperties
    &NiagaraLightRendererProperties
   ]
  }
  {emitterName=Flare, rendererClasses [&NiagaraSpriteRendererProperties]}
  {
   bEnabled=false
   emitterName=HeadLight001
   rendererClasses [
    &NiagaraMeshRendererProperties
    &NiagaraSpriteRendererProperties
    &NiagaraLightRendererProperties
   ]
  }
 ]
 userVariables []
}
