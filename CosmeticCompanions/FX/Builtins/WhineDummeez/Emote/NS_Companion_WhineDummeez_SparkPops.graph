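# Niagara effect for the WhineDummeez cosmetic companion's emote: popping sparks with smoke, spawned on a skeletal mesh via a shared system-level reader.
# LongBurst and ShortBurst fire three staggered randomized bursts of speed-stretched sprites that cool from overbright white-yellow to red.
# Long sparks fall and flicker out, while short ones stop quickly under heavy drag.
# Smoke combines a burst with tapering spawn rate, growing, fading and driving dynamic material parameters.
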NiagaraSystem /CosmeticCompanions/FX/Builtins/WhineDummeez/Emote/NS_Companion_WhineDummeez_SparkPops
dependencies {
 usedDataInterfaces [&NiagaraDataInterfaceCurve, &NiagaraDataInterfaceColorCurve, &NiagaraDataInterfaceSkeletalMesh]
 usedDynamicInputs []
 usedModules [
  &SystemState
  &self:SystemSpawnScript.SystemScriptSource.SystemScriptGraph.NiagaraNodeAssignment_0.SetVariables_#1
 ]
}
emitters {
 LongBurst {
  emitterUpdateScript [
   &EmitterState(Life Cycle Mode=::NewEnumerator1, Loop Behavior=::NewEnumerator1, Loop Duration=1, UseLoopDelay=false)
   &SpawnBurst_Instantaneous(
    Spawn Count=&RandomRangeInt(Minimum=9, Maximum=12)
    Use Spawn Probability=false
    UseLoopCountLimit=false
   )
   &SpawnBurst_Instantaneous(
    Spawn Count=&RandomRangeInt(Minimum=7, Maximum=10)
    Spawn Time=&RandomRangeFloat(Minimum=0.1, Maximum=0.3)
    Use Spawn Probability=false
    UseLoopCountLimit=false
   ) {module=SpawnBurst_Instantaneous001}
   &SpawnBurst_Instantaneous(
    Spawn Count=&RandomRangeInt(Minimum=5, Maximum=7)
    Spawn Time=&RandomRangeFloat(Minimum=0.4, Maximum=0.5)
    Use Spawn Probability=false
    UseLoopCountLimit=false
   ) {module=SpawnBurst_Instantaneous002}
  ]
  particleSpawnScript [
   &InitializeParticle(
    AdjustAlpha=true
    AdjustHue=false
    AdjustSaturation=false
    AdjustValue=true
    Color Mode=::NewEnumerator3
    Lifetime Max=1
    Lifetime Min=0.4
    Lifetime Mode=::NewEnumerator1
    Sprite Rotation Mode=::NewEnumerator2
    Sprite Size Mode=::NewEnumerator1
    Uniform Sprite Size Max=2.5
    Uniform Sprite Size Min=2
    UsePositionOffset=false
    Value Range:Vector2f=<0.8 5>
    Write Material Random=true
   )
   &SkeletalMeshLocation(MaskSpawnGroup=false, OverrideRandomSeed=false, Skeletal Mesh=$System.Skeletal Mesh Reader)
   &ShapeLocation(MaskSpawnGroup=false, Override Seed=false, Sphere Radius=5, UseOwnerScale=true)
   &AddVelocity(
    Module.Use Falloff Distance=false
    Origin Offset=&RandomRangeVector(Minimum=<-0.1 -0.1 -0.1>, Maximum=<0.1 0.1 -0.5>)
    Origin Offset Coordinate Space=::Simulation
    Override Seed=false
    Velocity Mode=::NewEnumerator1
    Velocity Origin=&AddVectorToPosition(Position=$Engine.Emitter.SimulationPosition, Vector=<0 0 10>)
    Velocity Speed=&RandomRangeFloat(Minimum=75, Maximum=300)
   )
   /SetVariables(Particles.Position=&AddVectorToPosition(Position=$Particles.Position, Vector=<0 0 5>)) {isSetParameters=true, module=SetVariables_#2}
  ]
  particleUpdateScript [
   &ParticleState()
   &GravityForce(Gravity=&RandomRangeVector(Minimum=<0 0 -400>, Maximum=<0 0 -980>))
   &Drag(
    Drag=&FloatFromCurve(
     FloatCurve=DataInterface() {propertyValues {Curve {
        defaultValue=0
        keys [
         {interpMode=RCIM_Cubic, tangentMode=RCTM_User, time=0, value=0}
         {interpMode=RCIM_Cubic, tangentMode=RCTM_User, time=0.5, value=1}
         {tangentMode=RCTM_User, time=1, value=0}
        ]
       }}}
     Scale Curve=&RandomRangeFloat(Minimum=5, Maximum=7)
    )
    Use Linear Drag=true
    Use Rotational Drag=false
   )
   &ScaleSpriteSize(Scale Factor=&Vector2DFromFloat(Value=&FloatFromCurve(
      FloatCurve=DataInterface() {propertyValues {Curve {keys [
          <at -0.0046 1.0113 RCIM_Cubic>
          {arriveTangent=-1.0036, interpMode=RCIM_Cubic, leaveTangent=-1.0036, time=0.7001, value=0.872}
          <at 1.0011 0.002>
         ]}}}
      Scale Curve=0.9
     )))
   &ScaleColor(
    Linear Color Curve=DataInterface() {propertyValues {
      AlphaCurve {keys [<at 0.0539 0.8673>,<at 1 0>]}
      BlueCurve {keys [<at 0.02 7.7622>,<at 0.1291 1>,<at 0.2563 0>,<at 0.4308 0.0006>,<at 1 0>]}
      GreenCurve {keys [<at 0.02 14.2001>,<at 0.1291 1>,<at 0.2563 0.8895>,<at 0.4308 0.1768>,<at 1 0.0502>]}
      RedCurve {keys [<at 0.02 15>,<at 0.1291 1>,<at 0.2563 1>,<at 0.4308 1>,<at 1 0.2917>]}
     }}
    Scale Mode=::NewEnumerator2
   )
   &SolveForcesAndVelocity(Clamp Velocity=false, Limit Acceleration=false)
   &ScaleSpriteSizeBySpeed(Max Scale Factor=<0.1 10>, Min Scale Factor=<1 0.75>, Velocity Threshold=500)
   &ScaleColor(
    Scale Alpha=&FloatFromCurve(FloatCurve=DataInterface() {propertyValues {Curve {
        defaultValue=1
        keys [
         {arriveTangent=9.0725, interpMode=RCIM_Cubic, leaveTangent=9.0725, tangentMode=RCTM_User, time=0, value=0}
         {arriveTangent=-2.5249, interpMode=RCIM_Cubic, leaveTangent=-2.5249, tangentMode=RCTM_User, time=0.1408, value=0.9847}
         {arriveTangent=-0.7988, interpMode=RCIM_Cubic, leaveTangent=-0.7988, time=0.5676, value=0.0941}
         {arriveTangent=-0.7205, interpMode=RCIM_Cubic, leaveTangent=-0.7205, time=0.6677, value=0.5637}
         {arriveTangent=-1.2036, interpMode=RCIM_Cubic, leaveTangent=-1.2036, time=0.7749, value=-0.0552}
         {arriveTangent=0.2454, interpMode=RCIM_Cubic, leaveTangent=0.2454, time=0.875, value=0.3143}
         {arriveTangent=-0.0405, interpMode=RCIM_Cubic, leaveTangent=-0.0405, tangentMode=RCTM_User, time=1, value=0}
        ]
       }}})
    ScaleA=true
    ScaleRGB=false
   ) {module=ScaleColor001}
  ]
  rendererClasses [&NiagaraSpriteRendererProperties]
  renderers [{rendererClass=&NiagaraSpriteRendererProperties}]
 }
 ShortBurst {
  emitterUpdateScript [
   &EmitterState(Life Cycle Mode=::NewEnumerator1, Loop Behavior=::NewEnumerator1, Loop Duration=1, UseLoopDelay=false)
   &SpawnBurst_Instantaneous(
    Spawn Count=&RandomRangeInt(Minimum=7, Maximum=13)
    Use Spawn Probability=false
    UseLoopCountLimit=false
   )
   &SpawnBurst_Instantaneous(
    Spawn Count=&RandomRangeInt(Minimum=5, Maximum=7)
    Spawn Time=&RandomRangeFloat(Minimum=0.1, Maximum=0.3)
    Use Spawn Probability=false
    UseLoopCountLimit=false
   ) {module=SpawnBurst_Instantaneous001}
   &SpawnBurst_Instantaneous(
    Spawn Count=&RandomRangeInt(Minimum=2, Maximum=3)
    Spawn Time=&RandomRangeFloat(Minimum=0.4, Maximum=0.45)
    Use Spawn Probability=false
    UseLoopCountLimit=false
   ) {module=SpawnBurst_Instantaneous002}
  ]
  particleSpawnScript [
   &InitializeParticle(
    AdjustAlpha=true
    AdjustHue=false
    AdjustSaturation=false
    AdjustValue=true
    Color Mode=::NewEnumerator3
    Lifetime Max=0.2
    Lifetime Min=0.175
    Lifetime Mode=::NewEnumerator1
    Sprite Rotation Mode=::NewEnumerator2
    Sprite Size Mode=::NewEnumerator1
    Uniform Sprite Size Max=2.5
    Uniform Sprite Size Min=2
    UsePositionOffset=false
    Value Range:Vector2f=<0.8 5>
    Write Material Random=true
   )
   &SkeletalMeshLocation(MaskSpawnGroup=false, OverrideRandomSeed=false, Skeletal Mesh=$System.Skeletal Mesh Reader)
   &ShapeLocation(
    MaskSpawnGroup=false
    Override Seed=false
    Sphere Radius=&RandomRangeFloat(Minimum=5, Maximum=8)
    UseOwnerScale=true
   )
   &AddVelocity(
    Module.Use Falloff Distance=false
    Origin Offset=&RandomRangeVector(Minimum=<-0.1 -0.1 -0.1>, Maximum=<0.1 0.1 -0.5>)
    Origin Offset Coordinate Space=::Simulation
    Override Seed=false
    Velocity Mode=::NewEnumerator1
    Velocity Origin=&AddVectorToPosition(Position=$Engine.Emitter.SimulationPosition, Vector=<0 0 10>)
    Velocity Speed=&RandomRangeFloat(Minimum=75, Maximum=300)
   )
   /SetVariables(Particles.Position=&AddVectorToPosition(Position=$Particles.Position, Vector=<0 0 5>)) {isSetParameters=true, module=SetVariables_#2}
  ]
  particleUpdateScript [
   &ParticleState()
   &Drag(
    Drag=&FloatFromCurve(
     FloatCurve=DataInterface() {propertyValues {Curve {
        defaultValue=0
        keys [
         {interpMode=RCIM_Cubic, tangentMode=RCTM_User, time=0, value=0}
         {interpMode=RCIM_Cubic, tangentMode=RCTM_User, time=0.5, value=1}
         {tangentMode=RCTM_User, time=1, value=0}
        ]
       }}}
     Scale Curve=&RandomRangeFloat(Minimum=25, Maximum=35)
    )
    Use Linear Drag=true
    Use Rotational Drag=false
   )
   &ScaleSpriteSize(Scale Factor=&Vector2DFromFloat(Value=&FloatFromCurve(
      FloatCurve=DataInterface() {propertyValues {Curve {keys [
          <at -0.0046 1.0113 RCIM_Cubic>
          {arriveTangent=-1.0036, interpMode=RCIM_Cubic, leaveTangent=-1.0036, time=0.7001, value=0.872}
          <at 1.0011 0.002>
         ]}}}
      Scale Curve=0.8
     )))
   &ScaleColor(
    Linear Color Curve=DataInterface() {propertyValues {
      AlphaCurve {keys [<at 0.0539 0.8673>,<at 1 0>]}
      BlueCurve {keys [<at 0.02 7.7622>,<at 0.1291 1>,<at 0.2563 0>,<at 0.4308 0.0006>,<at 1 0>]}
      GreenCurve {keys [<at 0.02 14.2001>,<at 0.1291 1>,<at 0.2563 0.8895>,<at 0.4308 0.1768>,<at 1 0.0502>]}
      RedCurve {keys [<at 0.02 15>,<at 0.1291 1>,<at 0.2563 1>,<at 0.4308 1>,<at 1 0.2917>]}
     }}
    Scale Mode=::NewEnumerator2
   )
   &SolveForcesAndVelocity(Clamp Velocity=false, Limit Acceleration=false)
   &ScaleSpriteSizeBySpeed(
    Max Scale Factor=&RandomRangeVector2D(Minimum=<0.75 8>, Maximum=<0.75 10>)
    Min Scale Factor=<1 0.75>
    Velocity Threshold=100
   )
   &ScaleColor(
    Scale Alpha=&FloatFromCurve(FloatCurve=DataInterface() {propertyValues {Curve {
        defaultValue=1
        keys [
         {interpMode=RCIM_Cubic, tangentMode=RCTM_User, time=0, value=1}
         {arriveTangent=-1.6146, interpMode=RCIM_Cubic, leaveTangent=-1.6146, tangentMode=RCTM_User, time=1, value=0}
        ]
       }}})
    ScaleA=true
    ScaleRGB=false
   ) {module=ScaleColor001}
   &CameraOffset(Camera Offset Amount=&RandomRangeFloat(Minimum=0, Maximum=5))
  ]
  rendererClasses [&NiagaraSpriteRendererProperties]
  renderers [{rendererClass=&NiagaraSpriteRendererProperties}]
 }
 Smoke {
  emitterUpdateScript [
   &EmitterState()
   &SpawnRate(
    SpawnRate=&FloatFromCurve(FloatCurve=DataInterface() {propertyValues {Curve {keys [<at 0 1>,<at 0.5 0>]}}}, CurveIndex=$Emitter.NormalizedLoopAge, Scale Curve=55)
    Use Spawn Probability=false
   )
   &SpawnBurst_Instantaneous(
    Spawn Count=&RandomRangeInt(Minimum=20, Maximum=25)
    Use Spawn Probability=false
    UseLoopCountLimit=false
   )
  ]
  particleSpawnScript [
   &InitializeParticle(
    AdjustAlpha=true
    AdjustHue=false
    AdjustSaturation=false
    AdjustValue=true
    Alpha Scale Range:Vector2f=<0.7 1.2>
    Color=&RandomRangeLinearColor(Minimum=<rgba 0.3854 0.3714 0.3714 0.3>, Maximum=<rgba 0.1771 0.1554 0.152 1>)
    Color Mode=::NewEnumerator3
    Lifetime Max=1
    Lifetime Min=0.25
    Lifetime Mode=::NewEnumerator1
    Sprite Rotation Angle=&RandomRangeFloat(Minimum=-25, Maximum=0)
    Sprite Rotation Mode=::NewEnumerator0
    Sprite Size Mode=::NewEnumerator1
    Sprite UV Mode=::NewEnumerator2
    Uniform Sprite Size Max=65
    Uniform Sprite Size Min=45
    UsePositionOffset=false
    Value Range:Vector2f=<0.5 1>
    Write Material Random=true
   )
   &AccelerationForce(Acceleration=&UniformRangedVector(Minimum=<0 0 0>, Maximum=<0 0 50>))
   &AddVelocity(Velocity=&UniformRangedVector(Minimum=<-50 -50 -50>, Maximum=<50 50 50>))
   &ApplyInitialForces()
   &SkeletalMeshLocation(MaskSpawnGroup=false, OverrideRandomSeed=false, Skeletal Mesh=$System.Skeletal Mesh Reader)
   /SetVariables(Particles.Position=&AddVectorToPosition(Position=$Particles.Position, Vector=<0 0 5>)) {isSetParameters=true}
  ]
  particleUpdateScript [
   &ParticleState()
   &Drag(Drag=5, Use Linear Drag=true, Use Rotational Drag=false)
   &GravityForce(Gravity=&RandomRangeVector(Minimum=<0 0 -10>, Maximum=<0 0 -100>))
   &SolveForcesAndVelocity(Clamp Velocity=false, Limit Acceleration=false)
   &ScaleSpriteSize(Scale Factor=&Vector2DFromFloat(Value=&FloatFromCurve(
      FloatCurve=DataInterface() {propertyValues {Curve {keys [
          {arriveTangent=1.2787, interpMode=RCIM_Cubic, leaveTangent=1.2787, leaveTangentWeight=0.5984, tangentMode=RCTM_User, time=0, value=0.1}
          <at 1 1 RCIM_Cubic>
         ]}}}
      Scale Curve=&RandomRangeFloat(Minimum=0.8, Maximum=1.4)
     )))
   /SetVariables(Particles.CloudSF:NiagaraFloat=&RandomRangeFloat(Minimum=0, Maximum=10)) {isSetParameters=true}
   /SetVariables(Particles.CloudSID:NiagaraFloat=&Lerp_Float(
     A=$Particles.CloudSF
     B=&Add_Float(A=$Particles.CloudSF, B=&RandomRangeFloat(Minimum=8, Maximum=15))
     Alpha=$Particles.NormalizedAge
    )) {isSetParameters=true}
   &DynamicMaterialParameters(
    Index 0 Param 1=&FloatFromCurve(FloatCurve=DataInterface() {propertyValues {Curve {keys [{arriveTangent=-2.1023, interpMode=RCIM_Cubic, leaveTangent=-2.1023, tangentMode=RCTM_User, time=0.0051, value=1}, <at 1.0011 0.002>]}}})
    Index 0 Param 4=&FloatFromCurve(FloatCurve=DataInterface() {propertyValues {Curve {
        defaultValue=1
        keys [
         {interpMode=RCIM_Cubic, tangentMode=RCTM_User, time=0, value=0}
         {arriveTangent=0.2137, interpMode=RCIM_Cubic, leaveTangent=0.2137, tangentMode=RCTM_User, time=1, value=1}
        ]
       }}})
    Module.Param0WriteEnabled=true
    Module.Param1WriteEnabled=false
    Module.Param2WriteEnabled=false
    Module.Param3WriteEnabled=true
   ) {module=DynamicMatParams}
   &CameraOffset(Camera Offset Amount=&RandomRangeFloat(Minimum=0, Maximum=50))
   &ScaleColor(
    Scale Alpha=&FloatFromCurve(FloatCurve=DataInterface() {propertyValues {Curve {keys [
         {arriveTangent=-3.1757, interpMode=RCIM_Cubic, leaveTangent=-3.1757, tangentMode=RCTM_User, time=0, value=1}
         {arriveTangent=-0.2376, interpMode=RCIM_Cubic, leaveTangent=-0.2376, tangentMode=RCTM_User, time=1, value=0}
        ]}}})
    Scale RGB=&VectorFromFloat(Value=&FloatFromCurve(FloatCurve=DataInterface() {propertyValues {Curve {keys [
          <at 0 1>
          {interpMode=RCIM_Cubic, leaveTangent=3.6059, tangentMode=RCTM_Break, time=0.0575, value=1.7315}
          {arriveTangent=-2.1334, interpMode=RCIM_Cubic, leaveTangent=-2.1334, tangentMode=RCTM_User, time=0.2298, value=1.7494}
          <at 1 0.5>
         ]}}}))
    ScaleA=true
    ScaleRGB=true
   )
  ]
  rendererClasses [&NiagaraSpriteRendererProperties]
  renderers [{rendererClass=&NiagaraSpriteRendererProperties}]
 }
}
summary {
 systemName=NS_Companion_WhineDummeez_SparkPops
 emitters [
  {emitterName=Smoke, rendererClasses [&NiagaraSpriteRendererProperties]}
  {emitterName=LongBurst, rendererClasses [&NiagaraSpriteRendererProperties]}
  {emitterName=ShortBurst, rendererClasses [&NiagaraSpriteRendererProperties]}
 ]
 userVariables []
}
